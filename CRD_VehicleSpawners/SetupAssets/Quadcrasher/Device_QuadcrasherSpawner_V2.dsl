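; Quadcrasher vehicle spawner device (V2) that exposes its wrap and fuel user options through overridden accessor functions of a parent spawner.
; Getters and setters map directly to user option variables for wrap support, fuel consumption, starting fuel and random starting fuel; fuel use multiplier has only a getter.
; The min and max random fuel setters call the parent before storing their values.
; The construction script only defers to the parent.
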(fn GetSupportsWraps
  (return (Variables|UserOptions|Basic|GetSupportsWraps)))

(fn GetConsumesFuel
  (return (Variables|UserOptions|GetFuel%20Consumption)))

(fn GetFuelUseMultiplier
  (return (Variables|UserOptions|GetFuel%20Use%20Multiplier)))

(fn GetMaxRandomStartingFuel
  (return (Variables|UserOptions|Advanced|GetMax%20Random%20Starting%20Fuel)))

(fn GetMinRandomStartingFuel
  (return (Variables|UserOptions|Advanced|GetMin%20Random%20Starting%20Fuel)))

(fn GetRandomStartingFuel
  (return (Variables|UserOptions|GetRandom%20Starting%20Fuel)))

(fn GetStartingFuelValue
  (return (Variables|UserOptions|GetStarting%20Fuel)))

(fn SetConsumeFuel
  (params (ConsumeFuel :type Boolean))
  (Variables|UserOptions|SetFuel%20Consumption ConsumeFuel))

(fn SetRandomStartingFuel
  (params (RandomStartingFuel :type Boolean))
  (Variables|UserOptions|SetRandom%20Starting%20Fuel RandomStartingFuel))

(fn SetStartingFuelValue
  (params (Starting%20Fuel%20Value :type Integer))
  (Variables|UserOptions|SetStarting%20Fuel Starting%20Fuel%20Value))

(fn SetSupportsWraps
  (params (Supports%20Wraps :type Boolean))
  (Variables|UserOptions|Basic|SetSupportsWraps Supports%20Wraps))

(fn SetMaxRandomStartingFuel
  (params (Max%20Fuel%20Percent :type Integer))
  (|CallParentFunction Max%20Fuel%20Percent)
  (Variables|UserOptions|Advanced|SetMax%20Random%20Starting%20Fuel Max%20Fuel%20Percent))

(fn SetMinRandomStartingFuel
  (params (Min%20Fuel%20Percent :type Integer))
  (|CallParentFunction Min%20Fuel%20Percent)
  (Variables|UserOptions|Advanced|SetMin%20Random%20Starting%20Fuel Min%20Fuel%20Percent))

(fn ConstructionScript
  (|CallParentFunction))
